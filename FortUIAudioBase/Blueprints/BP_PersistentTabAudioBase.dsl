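; Base Blueprint for persistent UI tab audio that plays music, a select sound and an optional loop on tab entry.
; Music reuses its component by toggling a bFadeOut parameter, respawning only when none exists or a watched MetaSound output reports it almost finished.
; The music component is also registered with the Music Event subsystem by tag while active.
; Tab exit fades out all three sounds; tab destruction stops only music and loop.
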(event EventOnTabEntered
  (|CallParentFunction)
  (CallFunction|PlayMusic)
  (CallFunction|PlayTabSelectSound)
  (CallFunction|PlayPersistentLoop))

(event EventOnTabExited
  (|CallParentFunction)
  (CallFunction|StopMusic)
  (CallFunction|StopTabSelectSound)
  (CallFunction|StopPersistentLoop))

(event EventOnTabDestroyed
  (|CallParentFunction)
  (CallFunction|StopMusic)
  (CallFunction|StopPersistentLoop))

(event Custom|Play%20Music
  (Utilities|FlowControl|Sequence
    (:then_0
      (if (Music|ShouldSpawnNewMusicAudioComponent)
        (bind _music (Variables|Music|GetMusic)
          (:then
            (bind _returnvalue (Audio|SpawnSound2D _music))
            (bind _output_get (Variables|Transient|SetMusic%20Audio%20Component _returnvalue))
            (MetaSoundOutput|WatchOutput (WorldSubsystems|GetMetaSoundOutputSubsystem) _output_get "bIsAlmostFinished" (delegate Custom|Set%20Is%20Almost%20Finished) "None" "None" :!class /Script/MetasoundEngine.MetaSoundOutputSubsystem))
          (:else))
        (else
          (bind _music_audio_component (Variables|Transient|GetMusic%20Audio%20Component)
            (:then
              (Audio|Parameter|SetBooleanParameter _music_audio_component "bFadeOut" false :!class /Script/AudioExtensions.AudioParameterControllerInterface))
            (:else)))))
    (:then_1
      (Utilities|FlowControl|Sequence
        (:then_0
          (Audio|AddEventbyComponent (GameInstanceSubsystems|GetMusicEventSubsystem) (Variables|Transient|GetMusic%20Audio%20Component) (Variables|Music|GetMusic%20Event%20Tag) "(GameplayTags=)" 0 ""))))))

(event Custom|Set%20Is%20Almost%20Finished
  (params (OutputName :type Name) (Output :type Meta%20Sound%20Output%20Structure%20%28by%20ref%29))
  (bind (_success _returnvalue) (MetaSoundOutput|GetBool Output))
  (Variables|Transient|SetIs%20Almost%20Finished _returnvalue))

(event Custom|Stop%20Music
  (bind _music_audio_component (Variables|Transient|GetMusic%20Audio%20Component)
    (:then
      (Audio|Parameter|SetBooleanParameter _music_audio_component "bFadeOut" true :!class /Script/AudioExtensions.AudioParameterControllerInterface)
      (Audio|RemoveEvent (GameInstanceSubsystems|GetMusicEventSubsystem) (Variables|Transient|GetMusic%20Audio%20Component)))
    (:else)))

(event Custom|Play%20Tab%20Select%20Sound
  (bind _returnvalue (Audio|SpawnSound2D (Variables|Sounds|GetTab%20Select%20Sound)))
  (Variables|Transient|SetTab%20Select%20Audio%20Component _returnvalue))

(event Custom|Stop%20Tab%20Select%20Sound
  (bind _tab_select_audio_component (Variables|Transient|GetTab%20Select%20Audio%20Component)
    (:then
      (Audio|Components|Audio|FadeOut _tab_select_audio_component 0.2 0.0 "Linear" :!class /Script/Engine.AudioComponent))
    (:else)))

(event Custom|Play%20Persistent%20Loop
  (bind _persistent_loop (Variables|Sounds|GetPersistent%20Loop)
    (:then
      (bind _returnvalue (Audio|SpawnSound2D _persistent_loop))
      (Variables|Default|SetPersistent%20Loop%20Audio _returnvalue))
    (:else)))

(event Custom|Stop%20Persistent%20Loop
  (bind _persistent_loop_audio (Variables|Default|GetPersistent%20Loop%20Audio)
    (:then
      (Audio|Components|Audio|FadeOut _persistent_loop_audio (Variables|Sounds|GetPersistent%20Loop%20Fade%20Out%20Time) 0.0 "Linear" :!class /Script/Engine.AudioComponent))
    (:else)))

(fn ShouldSpawnNewMusicAudioComponent
  (Utilities|FlowControl|Sequence
    (:then_0
      (Variables|Transient|GetMusic%20Audio%20Component
        (:then)
        (:else
          (return true))))
    (:then_1
      (Utilities|FlowControl|Sequence
        (:then_0
          (Variables|Transient|GetIs%20Almost%20Finished
            (:then
              (return true))
            (:else)))))))
