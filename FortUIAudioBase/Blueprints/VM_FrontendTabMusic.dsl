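; Frontend tab music controller that plays a MetaSound which follows the selected UI tab.
; Start Music spawns a 2D component if none exists or the MetaSound's watched almost-finished output is set, otherwise cancels the fade-out, then registers a music event.
; Tab names map to an "Index" parameter, cached and reapplied to new components to restore sub-tab selection.
; Fade Out Music sets the fade parameter and removes the event, keeping the component for reuse.
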(event Custom|Start%20Music
  (Utilities|FlowControl|Sequence
    (:then_0
      (if (Music|ShouldSpawnNewMusicAudioComponent)
        (bind _music (Variables|Music|GetMusic)
          (:then
            (bind _returnvalue (Audio|SpawnSound2D (Utilities|GetOuterObject self) _music))
            (bind _output_get (Variables|Transient|SetMusic%20Audio%20Component _returnvalue))
            (MetaSoundOutput|WatchOutput (WorldSubsystems|GetMetaSoundOutputSubsystem (Utilities|GetOuterObject self)) _output_get "bIsAlmostFinished" (delegate Custom|Set%20Is%20Almost%20Finished) "None" "None" :!class /Script/MetasoundEngine.MetaSoundOutputSubsystem)
            (Music|SetTabIndex (Variables|Transient|GetTab%20Index)) ; The tab index needs to be passed into the AC in cases where the player selects a sub-tab, leaves the main tab, and comes back to the main tab.
          )
          (:else))
        (else
          (bind _music_audio_component (Variables|Transient|GetMusic%20Audio%20Component)
            (:then
              (Audio|Parameter|SetBooleanParameter _music_audio_component "bFadeOut" false :!class /Script/AudioExtensions.AudioParameterControllerInterface))
            (:else)))))
    (:then_1
      (Utilities|FlowControl|Sequence
        (:then_0
          (Audio|AddEventbyComponent (GameInstanceSubsystems|GetMusicEventSubsystem (Utilities|GetOuterObject self)) (Variables|Transient|GetMusic%20Audio%20Component) (Variables|Music|GetMusic%20Event%20Tag) "(GameplayTags=)" 0 ""))))))

(event Custom|Set%20Is%20Almost%20Finished
  (params (OutputName :type Name) (Output :type Meta%20Sound%20Output%20Structure%20%28by%20ref%29))
  (bind (_success _returnvalue) (MetaSoundOutput|GetBool Output))
  (Variables|Transient|SetIs%20Almost%20Finished _returnvalue))

(event Custom|Fade%20Out%20Music
  (bind _music_audio_component (Variables|Transient|GetMusic%20Audio%20Component)
    (:then
      (Audio|Parameter|SetBooleanParameter _music_audio_component "bFadeOut" true :!class /Script/AudioExtensions.AudioParameterControllerInterface)
      (Audio|RemoveEvent (GameInstanceSubsystems|GetMusicEventSubsystem (Utilities|GetOuterObject self)) (Variables|Transient|GetMusic%20Audio%20Component)))
    (:else)))

(fn SetTabIndex
  (params (Selected%20Tab%20Index :type Integer))
  (Variables|Transient|SetTab%20Index Selected%20Tab%20Index)
  (bind _music_audio_component (Variables|Transient|GetMusic%20Audio%20Component)
    (:then
      (Audio|Parameter|SetIntegerParameter _music_audio_component "Index" (Variables|Transient|GetTab%20Index) :!class /Script/AudioExtensions.AudioParameterControllerInterface))
    (:else)))

(fn SetTabName
  (params (Name :type Name))
  (bind _tabs (Variables|Music|GetTabs))
  (bind _output_get (Variables|Transient|SetTab%20Name Name))
  (Music|SetTabIndex (Math|Integer|SelectInt (Utilities|Map|Find _tabs _output_get) 0 (.ReturnValue (Utilities|Map|Find _tabs _output_get))))
  (Variables|Debug|GetPrint%20Tab%20ID%20on%20Change
    (:then
      (Development|PrintText self (Utilities|Text|FormatText "NSLOCTEXT(\"[DE7C5CB408E78A8F98EB63F2EA2F631F]\", \"281AA57D46868DCEDA366ABB3C3B4E6B\", \"Currently selected TabID: {id}, Index: {index}\")" (Variables|Transient|GetTab%20Name) (Math|Integer|SelectInt (Utilities|Map|Find _tabs _output_get) 0 (.ReturnValue (Utilities|Map|Find _tabs _output_get))))))
    (:else)))

(fn ShouldSpawnNewMusicAudioComponent
  (Utilities|FlowControl|Sequence
    (:then_0
      (Variables|Transient|GetMusic%20Audio%20Component
        (:then)
        (:else
          (return true))))
    (:then_1
      (Utilities|FlowControl|Sequence
        (:then_0
          (Variables|Transient|GetIs%20Almost%20Finished
            (:then
              (return true))
            (:else)))))))
